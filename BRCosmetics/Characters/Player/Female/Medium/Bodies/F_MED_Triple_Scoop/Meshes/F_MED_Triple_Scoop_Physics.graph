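# Physics asset for the Triple Scoop female medium player body cosmetic, setting up simulation bodies for its four pack bones.
# Capsule bodies on the left and right dyn_packA and dyn_packB bones are each constrained directly to a small spine_04 anchor capsule.
# Tight swing limits, tuned per bone, and minimal twist keep any pack movement to a slight wobble around the upper spine.
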PhysicsAsset /BRCosmetics/Characters/Player/Female/Medium/Bodies/F_MED_Triple_Scoop/Meshes/F_MED_Triple_Scoop_Physics
bodies {
 dyn_packA_l [{shapeType=Capsule, length=3.4997, radius=6.7207, center=<-0.0112 -0.0045 -0.2101>, rotation=<rot 0 -172.6342 -79.5819>}]
 dyn_packA_r [{shapeType=Capsule, length=5.7453, radius=2.4606, center=<0.0288 0.088 -0.4953>, rotation=<rot 0 172.8534 -71.6264>}]
 dyn_packB_l [{shapeType=Capsule, length=6.25, radius=4.7106, center=<-0.0415 0.1879 -1.2458>, rotation=<rot 0 -172.8534 -79.3823>}]
 dyn_packB_r [{shapeType=Capsule, length=4.5, radius=6.2106, center=<0.0174 -0.0031 -0.2191>, rotation=<rot 0 172.8526 -77.6649>}]
 spine_04 [{shapeType=Capsule, length=1, radius=1}]
}
constraints [
 {bone1Name=dyn_packA_l, bone2Name=spine_04, swing1LimitDegrees=5, swing2LimitDegrees=5, twistLimitDegrees=1}
 {bone1Name=dyn_packB_l, bone2Name=spine_04, swing1LimitDegrees=10, swing2LimitDegrees=10, twistLimitDegrees=1}
 {bone1Name=dyn_packA_r, bone2Name=spine_04, swing1LimitDegrees=8, swing2LimitDegrees=8, twistLimitDegrees=1}
 {bone1Name=dyn_packB_r, bone2Name=spine_04, swing1LimitDegrees=4, swing2LimitDegrees=6, twistLimitDegrees=1}
]
