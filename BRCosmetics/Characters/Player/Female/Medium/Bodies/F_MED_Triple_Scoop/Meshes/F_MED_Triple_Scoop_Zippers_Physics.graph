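# Physics asset for the four dangling zippers (front and back, left and right) on the F_MED_Triple_Scoop female player cosmetic body.
# Each zipper bone has a capsule body constrained to its own small driver-bone capsule, forming four independent two-body pairs.
# Tight angular limits restrict swing and nearly lock twist, with back zippers allowed more movement than front ones.
# Zipper capsule sizes differ between left and right.
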PhysicsAsset /BRCosmetics/Characters/Player/Female/Medium/Bodies/F_MED_Triple_Scoop/Meshes/F_MED_Triple_Scoop_Zippers_Physics
bodies {
 dyn_driver_zipper_bck_l [{shapeType=Capsule, length=0.505, radius=0.505, rotation=<rot 0 -90 0>}]
 dyn_driver_zipper_bck_r [{shapeType=Capsule, length=0.505, radius=0.505, rotation=<rot 0 -90 0>}]
 dyn_driver_zipper_f_l [{shapeType=Capsule, length=0.505, radius=0.505, rotation=<rot 0 -90 0>}]
 dyn_driver_zipper_f_r [{shapeType=Capsule, length=0.505, radius=0.505, rotation=<rot 0 -90 0>}]
 dyn_zipper_bck_l [{shapeType=Capsule, length=2.0141, radius=1.1539, center=<1.6095 0.0313 0.014>, rotation=<rot 0 89.5908 -88.3381>}]
 dyn_zipper_bck_r [{shapeType=Capsule, length=1.7641, radius=2.4032, center=<1.6096 0.0311 -0.0112>, rotation=<rot 0 -90.3789 -88.3314>}]
 dyn_zipper_f_l [{shapeType=Capsule, length=1.25, radius=1.75, center=<1 0 0>, rotation=<rot 90 0 0>}]
 dyn_zipper_f_r [{shapeType=Capsule, length=1.75, radius=2.75, center=<1 0 0>, rotation=<rot 90 0 0>}]
}
constraints [
 {bone1Name=dyn_zipper_f_l, bone2Name=dyn_driver_zipper_f_l, swing1LimitDegrees=10, swing2LimitDegrees=5, twistLimitDegrees=1}
 {bone1Name=dyn_zipper_bck_l, bone2Name=dyn_driver_zipper_bck_l, swing1LimitDegrees=15, swing2LimitDegrees=25, twistLimitDegrees=3}
 {bone1Name=dyn_zipper_f_r, bone2Name=dyn_driver_zipper_f_r, swing1LimitDegrees=10, swing2LimitDegrees=5, twistLimitDegrees=1}
 {bone1Name=dyn_zipper_bck_r, bone2Name=dyn_driver_zipper_bck_r, swing1LimitDegrees=15, swing2LimitDegrees=20, twistLimitDegrees=3}
]
