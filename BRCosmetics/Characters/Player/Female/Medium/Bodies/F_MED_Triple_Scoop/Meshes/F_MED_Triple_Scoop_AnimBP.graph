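# Secondary-motion AnimBlueprint for the F_MED_Triple_Scoop player cosmetic body, layered on the pose copied from its attached parent mesh.
# Constraints, Flap nodes and additive bone offsets drive zipper, glute and shell bones, then two RigidBody passes simulate physics, one for zippers.
# Thigh twist sets glute correction strength, damped above a speed threshold, and crouch events toggle front zipper offsets.
# On customization load, it hides any tail when a long backbling is worn.
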AnimBlueprint /BRCosmetics/Characters/Player/Female/Medium/Bodies/F_MED_Triple_Scoop/Meshes/F_MED_Triple_Scoop_AnimBP
target_skeleton=&/F_MED_Triple_Scoop_Skeleton
graphs [
 AnimGraph {nodes [
   n1 = Get crouch_alpha() -> (crouch_alpha:double)
   n2 = "Constraint - dyn_parent_zipper_f_l (PPP)"!(
    ComponentPose=Local To Component(LocalPose=Copy Pose From Mesh() {bUseAttachedParent=true})
    Alpha=1
    ConstraintWeights_0=0.7289
    ConstraintWeights_1:float=0.2665
    ConstraintWeights_2:float=0.0046
   ) -> (Pose) {
    alphaBoolBlend {blendOption=Linear}
    boneToModify {boneName=dyn_parent_zipper_f_l}
    constraintSetup [{targetBone {boneName=pelvis}}, {targetBone {boneName=thigh_twist_01_l}}, {targetBone {boneName=thigh_l}}]
    constraintWeights [1,1,1]
   }
   n3 = "Constraint - dyn_parent_zipper_f_r (PPP)"!(ComponentPose=@n2, Alpha=1, ConstraintWeights_0=0.0077, ConstraintWeights_1:float=0.6434, ConstraintWeights_2:float=0.3489) -> (Pose) {
    alphaBoolBlend {blendOption=Linear}
    boneToModify {boneName=dyn_parent_zipper_f_r}
    constraintSetup [{targetBone {boneName=thigh_r}}, {targetBone {boneName=pelvis}}, {targetBone {boneName=thigh_twist_01_r}}]
    constraintWeights [1,1,1]
   }
   n4 = "Transform (Modify) Bone - Bone: dyn_driver_zipper_bck_l"(ComponentPose=@n3, Alpha=1, Translation="0,0.1,0", Rotation="0,0,15", Scale="1,1,1") {
    rotationMode=BMM_Additive
    rotationSpace=BCS_BoneSpace
    translationMode=BMM_Additive
    translationSpace=BCS_BoneSpace
    alphaBoolBlend {blendOption=Linear}
    boneToModify {boneName=dyn_driver_zipper_bck_l}
    rotation {roll=15}
    translation {y=0.1}
   }
   n5 = "Transform (Modify) Bone - Bone: dyn_driver_zipper_bck_r"(ComponentPose=@n4, Alpha=1, Translation="0,0.1,0", Rotation="0,0,-15", Scale="1,1,1") {
    rotationMode=BMM_Additive
    rotationSpace=BCS_BoneSpace
    translationMode=BMM_Additive
    translationSpace=BCS_BoneSpace
    alphaBoolBlend {blendOption=Linear}
    boneToModify {boneName=dyn_driver_zipper_bck_r}
    rotation {roll=-15}
    translation {y=0.1}
   }
   n6 = Flap Node - Bone: dyn_driver_zipper_f_l!(ComponentPose:Component Space Pose Link Structure=@n5, Alpha:float=1) -> (Pose:Component Space Pose Link Structure) {
    is=title FortAnimGraphNode_Flap
    bUseClamp=true
    targetClampMax=10
    targetClampMin=-10
    targetComponent=RotationZ
    alphaBoolBlend {blendOption=Linear}
    sourceDrives [{
      bUseRange=false
      drivingCurve=null
      multiplier=0.5
      rangeMax=1
      rangeMin=-1
      remappedMax=1
      remappedMin=0
      sourceComponent=RotationZ
      useQuaternion=true
      sourceBone {boneName=thigh_l}
     }]
    targetBone {boneName=dyn_driver_zipper_f_l}
   }
   n7 = Flap Node - Bone: dyn_driver_zipper_f_r!(ComponentPose:Component Space Pose Link Structure=@n6, Alpha:float=1) -> (Pose:Component Space Pose Link Structure) {
    is=title FortAnimGraphNode_Flap
    bUseClamp=true
    targetClampMax=10
    targetClampMin=-10
    targetComponent=RotationZ
    alphaBoolBlend {blendOption=Linear}
    sourceDrives [{
      bUseRange=false
      drivingCurve=null
      multiplier=-0.5
      rangeMax=1
      rangeMin=-1
      remappedMax=1
      remappedMin=0
      sourceComponent=RotationZ
      useQuaternion=true
      sourceBone {boneName=thigh_r}
     }]
    targetBone {boneName=dyn_driver_zipper_f_r}
   }
   n8 = "Transform (Modify) Bone - Bone: dyn_parent_zipper_f_l"(ComponentPose=@n7, Alpha=@n1, Translation="-0.486,-6.501,-3.000", Rotation="0,0,0", Scale="1,1,1") {
    rotationMode=BMM_Additive
    rotationSpace=BCS_BoneSpace
    translationMode=BMM_Additive
    translationSpace=BCS_BoneSpace
    alphaBoolBlend {blendOption=Linear}
    boneToModify {boneName=dyn_parent_zipper_f_l}
    translation=<-0.486 -6.501 -3>
   }
   n9 = "Transform (Modify) Bone - Bone: dyn_parent_zipper_f_r"(ComponentPose=@n8, Alpha=@n1, Translation="-0.000,-1,8", Rotation="0,0,0", Scale="1,1,1") {
    rotationMode=BMM_Additive
    rotationSpace=BCS_BoneSpace
    translationMode=BMM_Additive
    translationSpace=BCS_BoneSpace
    alphaBoolBlend {blendOption=Linear}
    boneToModify {boneName=dyn_parent_zipper_f_r}
    translation {y=-1, z=8}
   }
   n10 = "Transform (Modify) Bone - Bone: dyn_glute_l"(ComponentPose=@n9, Alpha=Get thigh_rot_l(), Translation="-3.750,0.875,-0.9", Rotation="0.000,0.000,-55", Scale="1.25,1.25,1.25") {
    rotationMode=BMM_Additive
    rotationSpace=BCS_BoneSpace
    scaleMode=BMM_Additive
    scaleSpace=BCS_BoneSpace
    translationMode=BMM_Additive
    translationSpace=BCS_BoneSpace
    alphaBoolBlend {blendOption=Linear}
    boneToModify {boneName=dyn_glute_l}
    rotation {roll=-55}
    scale=<1.25 1.25 1.25>
    translation=<-3.75 0.875 -0.9>
   }
   n11 = "Transform (Modify) Bone - Bone: dyn_glute_r"(ComponentPose=@n10, Alpha=Get thigh_rot_r(), Translation="-3.75,0,0", Rotation="0.000,0.000,-55", Scale="1.3,1.3,1.3") {
    rotationMode=BMM_Additive
    rotationSpace=BCS_BoneSpace
    scaleMode=BMM_Additive
    scaleSpace=BCS_BoneSpace
    translationMode=BMM_Additive
    translationSpace=BCS_BoneSpace
    alphaBoolBlend {blendOption=Linear}
    boneToModify {boneName=dyn_glute_r}
    rotation {roll=-55}
    scale=<1.3 1.3 1.3>
    translation {x=-3.75}
   }
   n12 = "Constraint - dyn_shell_01_l (P)"!(ComponentPose=@n11, Alpha=1, ConstraintWeights_0=1, ConstraintWeights_1:float=0) -> (Pose) {
    alphaBoolBlend {blendOption=Linear}
    boneToModify {boneName=dyn_shell_01_l}
    constraintSetup [{targetBone {boneName=clavicle_l}}]
    constraintWeights [1,0]
   }
   n13 = "Constraint - dyn_shell_02_l (P)"!(ComponentPose=@n12, Alpha=1, ConstraintWeights_0=1, ConstraintWeights_1:float=0) -> (Pose) {
    alphaBoolBlend {blendOption=Linear}
    boneToModify {boneName=dyn_shell_02_l}
    constraintSetup [{targetBone {boneName=clavicle_l}}]
    constraintWeights [1,0]
   }
   n14 = "Constraint - dyn_shell_03_l (P)"!(ComponentPose=@n13, Alpha=1, ConstraintWeights_0=1, ConstraintWeights_1:float=0) -> (Pose) {
    alphaBoolBlend {blendOption=Linear}
    boneToModify {boneName=dyn_shell_03_l}
    constraintSetup [{targetBone {boneName=clavicle_l}}]
    constraintWeights [1,0]
   }
   n15 = "Constraint - dyn_shell_04_l (P)"!(ComponentPose=@n14, Alpha=1, ConstraintWeights_0=1, ConstraintWeights_1:float=0) -> (Pose) {
    alphaBoolBlend {blendOption=Linear}
    boneToModify {boneName=dyn_shell_04_l}
    constraintSetup [{targetBone {boneName=clavicle_l}}]
    constraintWeights [1,0]
   }
   n16 = "Constraint - dyn_shell_01_r (P)"!(ComponentPose=@n15, Alpha=1, ConstraintWeights_0=1, ConstraintWeights_1:float=0) -> (Pose) {
    alphaBoolBlend {blendOption=Linear}
    boneToModify {boneName=dyn_shell_01_r}
    constraintSetup [{targetBone {boneName=clavicle_r}}]
    constraintWeights [1,0]
   }
   n17 = "Constraint - dyn_shell_02_r (P)"!(ComponentPose=@n16, Alpha=1, ConstraintWeights_0=1, ConstraintWeights_1:float=0) -> (Pose) {
    alphaBoolBlend {blendOption=Linear}
    boneToModify {boneName=dyn_shell_02_r}
    constraintSetup [{targetBone {boneName=clavicle_r}}]
    constraintWeights [1,0]
   }
   n18 = "Constraint - dyn_shell_03_r (P)"!(ComponentPose=@n17, Alpha=1, ConstraintWeights_0=1, ConstraintWeights_1:float=0) -> (Pose) {
    alphaBoolBlend {blendOption=Linear}
    boneToModify {boneName=dyn_shell_03_r}
    constraintSetup [{targetBone {boneName=clavicle_r}}]
    constraintWeights [1,0]
   }
   n19 = "Constraint - dyn_shell_04_r (PP)"!(ComponentPose=@n18, Alpha=1, ConstraintWeights_0=0.0124, ConstraintWeights_1:float=0.9876) -> (Pose) {
    alphaBoolBlend {blendOption=Linear}
    boneToModify {boneName=dyn_shell_04_r}
    constraintSetup [{targetBone {boneName=upperarm_twist_01_r}}, {targetBone {boneName=clavicle_r}}]
    constraintWeights [1,1]
   }
   n20 = RigidBody(ComponentPose=@n19) {
    bClampLinearTranslationLimitToRefPose=true
    bOverrideWorldGravity=true
    overridePhysicsAsset=&/F_MED_Triple_Scoop_Physics
    simulationSpace=BaseBoneSpace
    alphaBoolBlend {blendOption=Linear}
    baseBoneRef {boneName=root}
    componentAppliedLinearAccClamp=<1100 1100 1100>
    componentLinearAccScale=<0.5 0.5 0.5>
    componentLinearVelScale=<0.5 0.5 0.5>
    overrideWorldGravity {z=-980}
   }
   n21 = RigidBody(ComponentPose=@n20) {
    bClampLinearTranslationLimitToRefPose=true
    bOverrideWorldGravity=true
    overridePhysicsAsset=&/F_MED_Triple_Scoop_Zippers_Physics
    simulationSpace=BaseBoneSpace
    alphaBoolBlend {blendOption=Linear}
    baseBoneRef {boneName=root}
    componentAppliedLinearAccClamp=<1100 1100 1100>
    componentLinearAccScale=<0.5 0.5 0.35>
    componentLinearVelScale=<0.5 0.5 0.35>
    overrideWorldGravity {z=-980}
   }
   Output Pose(Result=Component To Local(ComponentPose=@n21)) {layerGroup=None}
  ]}
 EventGraph {nodes [
   Event BlueprintUpdateAnimation()
   > Get!(execute:Exec) -> (then:Exec, else:Exec, Skeletal_Mesh:Skeletal Mesh Component Object Reference)
   > Sequence!(execute) -> (then_0, then_1, then_2:Exec, then_3:Exec)
   n1 = Get Skeletal_Mesh() -> (Skeletal_Mesh:Skeletal Mesh Component Object Reference)
   n2 = GetAlphaFromReferenceTransformDelta(SkeletalMeshComponent=@n1, BoneName=thigh_l, BaseName=pelvis, DeltaComponentType=QuaternionTwist, InputRange=(X=20,Y=100), TwistAxis="0,0,1")
   > Set thigh_rot_l(thigh_rot_l=@n2)
   n3 = GetAlphaFromReferenceTransformDelta(SkeletalMeshComponent=@n1, BoneName=thigh_r, BaseName=pelvis, DeltaComponentType=QuaternionTwist, InputRange=(X=20,Y=100), TwistAxis="0,0,1")
   > Set thigh_rot_r(thigh_rot_r=@n3)
   > Branch(Condition="float > float"(A=Get PawnSpeed2D(), B=100) {is=title K2Node_CallFunction})
   > Set thigh_rot_l(thigh_rot_l=float * float(A=@n2, B=0.35))
   > Set thigh_rot_r(thigh_rot_r=float * float(A=@n3, B=0.35))
   Event OnBeginCrouching()
   > Set zipper_l_trans(zipper_l_trans:Vector="5,0.1, 0") -> (Output_Get:Vector)
   > Set crouch_alpha(crouch_alpha:double=1) -> (Output_Get:double)
   Event BlueprintInitializeAnimation()
   > Set Skeletal_Mesh(Skeletal_Mesh:Skeletal Mesh Component Object Reference=GetOwnerSkeletalMeshComponent()) -> (Output_Get:Skeletal Mesh Component Object Reference)
   Event OnCharacterCustomizationLoaded()
   > Set BodyMesh(BodyMesh=GetOwnerSkeletalMeshComponent())
   > n4 = Cast To PlayerPawn_Athena(Object=TryGetPawnOwner())
   > n5 = Set PawnOwner(PawnOwner:Player Pawn Athena Object Reference=@n4) -> (Output_Get:Player Pawn Athena Object Reference)
   > Set TailMesh(TailMesh:Skeletal Mesh Component Object Reference=GetSkeletalMeshForPartType(self=@n5, PartType=MiscOrTail)) -> (Output_Get:Skeletal Mesh Component Object Reference)
   n6 = CheckForBackbling()
   > Set isWearingBackbling(isWearingBackbling=@n6)
   > Set isWearingCape(isWearingCape=@n6.bIsBackblingLong)
   > Set characterHasTail(characterHasTail:bool=CheckForTail!(self:Custom Character Part Anim Instance Object Reference) -> (bHasTail:bool)) -> (Output_Get:bool)
   > Branch(Condition=Get characterHasTail() -> (characterHasTail:bool))
   > n7 = Branch(Condition=Get isWearingCape() -> (isWearingCape:bool))
   n8 = Get TailMesh() -> (TailMesh:Skeletal Mesh Component Object Reference)
   > SetVisibility!(execute, self:Scene Component Object Reference=@n8, bNewVisibility:bool=false, bPropagateToChildren:bool=false) -> (then)
   > Set cape and tail combo(cape and tail combo:bool=false) -> (Output_Get:bool)
   SetVisibility(execute=@n7.else, self=@n8, bNewVisibility=true, bPropagateToChildren=false)
   > Set cape and tail combo(cape and tail combo=true)
   Event OnEndCrouching()
   > Set zipper_l_trans(zipper_l_trans="0,0.1, 0")
   > Set crouch_alpha(crouch_alpha=0)
  ]}
]
variables [
 BodyMesh {type=SkeletalMeshComponent}
 cape and tail combo {type=boolean}
 characterHasTail {type=boolean}
 crouch_alpha {type=number}
 isWearingBackbling {type=boolean}
 isWearingCape {type=boolean}
 PawnOwner {type=PlayerPawn_Athena_C}
 Skeletal_Mesh {type=SkeletalMeshComponent}
 TailMesh {type=SkeletalMeshComponent}
 thigh_rot_l {type=number}
 thigh_rot_r {type=number}
 zipper_l_trans {type=Vector, default=<0 0.1 0>}
]
