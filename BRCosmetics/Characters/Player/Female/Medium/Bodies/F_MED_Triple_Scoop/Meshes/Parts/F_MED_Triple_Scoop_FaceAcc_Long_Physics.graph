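# Physics asset for the long face accessory part of the female medium Triple Scoop character, set up around its dynamic hair bones.
# Capsule bodies cover a three-bone hair chain (dyn_hair_parent, 01, 02) plus head, neck_01 and spine_05.
# Constraints link the chain from hair tip through head and neck to upper spine.
# The upper hair joints have narrow swing and very tight twist limits, while the tip, head and neck joints rotate freely.
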PhysicsAsset /BRCosmetics/Characters/Player/Female/Medium/Bodies/F_MED_Triple_Scoop/Meshes/Parts/F_MED_Triple_Scoop_FaceAcc_Long_Physics
bodies {
 dyn_hair_01 [{shapeType=Capsule, length=2.25, radius=2.005, center=<2 0 0>, rotation=<rot 0 -90 90>}]
 dyn_hair_02 [{shapeType=Capsule, length=5.75, radius=3.255, center=<0 1.714 -5.1686>, rotation=<rot -16.972 -89.9999 0>}]
 dyn_hair_parent [{shapeType=Capsule, length=2.25, radius=2.005, center=<1.918 0 -0.2894>, rotation=<rot -81.4193 -179.9999 180>}]
 head [{shapeType=Capsule, length=5, radius=6.75, rotation=<rot 90 -0.0003 -0.0002>}]
 neck_01 [{shapeType=Capsule, length=6.25, radius=4.5, center=<0.3642 1.1246 0.0048>, rotation=<rot 89.3571 -179.9998 -162.0508>}]
 spine_05 [{shapeType=Capsule, length=15, radius=8.25, center=<2.3703 -1.851 0.0001>}]
}
constraints [
 {
  bone1Name=head
  bone2Name=neck_01
  swing1LimitDegrees=45
  swing1Motion=Free
  swing2LimitDegrees=45
  swing2Motion=Free
  twistLimitDegrees=45
  twistMotion=Free
 }
 {bone1Name=dyn_hair_parent, bone2Name=head, swing1LimitDegrees=25, swing2LimitDegrees=25, twistLimitDegrees=2}
 {bone1Name=dyn_hair_01, bone2Name=dyn_hair_parent, swing1LimitDegrees=25, swing2LimitDegrees=25, twistLimitDegrees=2}
 {
  bone1Name=neck_01
  bone2Name=spine_05
  swing1LimitDegrees=45
  swing1Motion=Free
  swing2LimitDegrees=45
  swing2Motion=Free
  twistLimitDegrees=45
  twistMotion=Free
 }
 {
  bone1Name=dyn_hair_02
  bone2Name=dyn_hair_01
  swing1LimitDegrees=45
  swing1Motion=Free
  swing2LimitDegrees=45
  swing2Motion=Free
  twistLimitDegrees=45
  twistMotion=Free
 }
]
