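# Physics asset for the Mint face accessory of the female medium Triple Scoop cosmetic body, defining left and right ponytail physics.
# Each ponytail is a four-capsule chain hung from the head through a small master body, with tight twist limits and swing limits that generally widen toward the tip.
# Capsules on head, neck, upper spine and clavicles form the body volume; their joints and the master-to-head joints leave rotation free.
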PhysicsAsset /BRCosmetics/Characters/Player/Female/Medium/Bodies/F_MED_Triple_Scoop/Meshes/Parts/F_MED_Triple_Scoop_Mint_FaceAcc_Physics1
bodies {
 clavicle_l [{shapeType=Capsule, length=6.755, radius=6, center=<7.2033 -2.4486 0>, rotation=<rot 0 -89.9998 89.9994>}]
 clavicle_r [{shapeType=Capsule, length=6.755, radius=6, center=<-7.2033 0 0>, rotation=<rot 0 -89.9998 -90.0006>}]
 dyn_ponytail_01_l [{shapeType=Capsule, length=4.0787, radius=3, center=<0.75 0 0>, rotation=<rot 90 0 0>}]
 dyn_ponytail_01_l_MSTR [{shapeType=Capsule, length=0.505, radius=0.505, rotation=<rot 0 -90 0>}]
 dyn_ponytail_01_r [{shapeType=Capsule, length=3.9877, radius=3, center=<-0.75 0 0>, rotation=<rot 90 0 0>}]
 dyn_ponytail_01_r_MSTR [{shapeType=Capsule, length=0.505, radius=0.505, rotation=<rot 0 -90 0>}]
 dyn_ponytail_02_l [{shapeType=Capsule, length=2, radius=3, center=<1 -0.0001 1.766>, rotation=<rot 90 0 0>}]
 dyn_ponytail_02_r [{shapeType=Capsule, length=2, radius=2, center=<-1 -0.0117 -0.7>, rotation=<rot 90 0 0>}]
 dyn_ponytail_03_l [{shapeType=Capsule, length=1.75, radius=3.25, center=<2 0 1.2012>, rotation=<rot 90 0 0>}]
 dyn_ponytail_03_r [{shapeType=Capsule, length=1.5, radius=2.5, center=<-2 -0.0117 -0.7>, rotation=<rot 90 0 0>}]
 dyn_ponytail_04_l [{shapeType=Capsule, length=1, radius=2.25, center=<3.1671 0 -0.182>, rotation=<rot 90 0 0>}]
 dyn_ponytail_04_r [{shapeType=Capsule, length=1.5, radius=2.25, center=<-2.6138 0 0>, rotation=<rot 90 0 0>}]
 head [{shapeType=Capsule, length=1, radius=1}]
 neck_01 [{shapeType=Capsule, length=4, radius=3.75, center=<-0.1384 1.876 0.0002>, rotation=<rot -90 -176.4237 -173.5752>}]
 spine_04 [{shapeType=Capsule, length=6.5, radius=12.5, center=<-0.5651 -0.0001 0>}]
 spine_05 [{shapeType=Capsule, length=8.25, radius=14.75, center=<-4.0465 -4.9197 -0.0001>, rotation=<rot 0 0.0014 0>}]
}
constraints [
 {
  bone1Name=dyn_ponytail_01_l_MSTR
  bone2Name=head
  swing1LimitDegrees=45
  swing1Motion=Free
  swing2LimitDegrees=45
  swing2Motion=Free
  twistLimitDegrees=45
  twistMotion=Free
 }
 {bone1Name=dyn_ponytail_01_l, bone2Name=dyn_ponytail_01_l_MSTR, swing1LimitDegrees=18, swing2LimitDegrees=10, twistLimitDegrees=5}
 {
  bone1Name=dyn_ponytail_01_r_MSTR
  bone2Name=head
  swing1LimitDegrees=45
  swing1Motion=Free
  swing2LimitDegrees=45
  swing2Motion=Free
  twistLimitDegrees=45
  twistMotion=Free
 }
 {bone1Name=dyn_ponytail_01_r, bone2Name=dyn_ponytail_01_r_MSTR, swing1LimitDegrees=10, swing2LimitDegrees=10, twistLimitDegrees=5}
 {bone1Name=dyn_ponytail_02_l, bone2Name=dyn_ponytail_01_l, swing1LimitDegrees=20, swing2LimitDegrees=10, twistLimitDegrees=5}
 {bone1Name=dyn_ponytail_03_l, bone2Name=dyn_ponytail_02_l, swing1LimitDegrees=23, swing2LimitDegrees=25, twistLimitDegrees=5}
 {bone1Name=dyn_ponytail_04_l, bone2Name=dyn_ponytail_03_l, swing1LimitDegrees=45, swing2LimitDegrees=30, twistLimitDegrees=5}
 {bone1Name=dyn_ponytail_02_r, bone2Name=dyn_ponytail_01_r, swing1LimitDegrees=10, swing2LimitDegrees=10, twistLimitDegrees=5}
 {bone1Name=dyn_ponytail_03_r, bone2Name=dyn_ponytail_02_r, swing1LimitDegrees=15, swing2LimitDegrees=25, twistLimitDegrees=5}
 {bone1Name=dyn_ponytail_04_r, bone2Name=dyn_ponytail_03_r, swing1LimitDegrees=45, swing2LimitDegrees=30, twistLimitDegrees=5}
 {
  bone1Name=head
  bone2Name=neck_01
  swing1LimitDegrees=45
  swing1Motion=Free
  swing2LimitDegrees=45
  swing2Motion=Free
  twistLimitDegrees=45
  twistMotion=Free
 }
 {
  bone1Name=neck_01
  bone2Name=spine_05
  swing1LimitDegrees=45
  swing1Motion=Free
  swing2LimitDegrees=45
  swing2Motion=Free
  twistLimitDegrees=45
  twistMotion=Free
 }
 {
  bone1Name=spine_05
  bone2Name=spine_04
  swing1LimitDegrees=45
  swing1Motion=Free
  swing2LimitDegrees=45
  swing2Motion=Free
  twistLimitDegrees=45
  twistMotion=Free
 }
 {
  bone1Name=clavicle_l
  bone2Name=spine_05
  swing1LimitDegrees=45
  swing1Motion=Free
  swing2LimitDegrees=45
  swing2Motion=Free
  twistLimitDegrees=45
  twistMotion=Free
 }
 {
  bone1Name=clavicle_r
  bone2Name=spine_05
  swing1LimitDegrees=45
  swing1Motion=Free
  swing2LimitDegrees=45
  swing2Motion=Free
  twistLimitDegrees=45
  twistMotion=Free
 }
]
